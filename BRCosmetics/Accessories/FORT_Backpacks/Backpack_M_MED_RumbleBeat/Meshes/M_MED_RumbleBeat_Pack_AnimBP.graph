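# AnimBlueprint for the RumbleBeat backpack that reveals barbell disks one by one as the player eliminates enemies.
# The EventGraph counts elimination anim-input events, or increments on timers in the frontend, and sets flags that zero-scale the hidden Barbell bones.
# In game, spine yaw relative to ref pose also rotates and shifts the pack bone.
# Blend-by-bool nodes choose between in-game, reactive frontend and static all-disks frontend poses.
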AnimBlueprint /BRCosmetics/Accessories/FORT_Backpacks/Backpack_M_MED_RumbleBeat/Meshes/M_MED_RumbleBeat_Pack_AnimBP
target_skeleton=&/M_MED_RumbleBeat_Pack_Skeleton
graphs [
 AnimGraph {nodes [
   n1 = "Transform (Modify) Bone - Bone: Barbell_03"!(
    ComponentPose=Local To Component(LocalPose=Use cached pose 'sim for Fe/game'())
    bAlphaBoolEnabled:bool=Get disk_1() -> (disk_1:bool)
    Translation="0,0,0"
    Rotation="0,0,0"
    Scale="0,0,0"
   ) -> (Pose) {
    alphaInputType=Bool
    scaleMode=BMM_Replace
    scaleSpace=BCS_BoneSpace
    translationMode=BMM_Additive
    translationSpace=BCS_BoneSpace
    boneToModify {boneName=Barbell_03}
    scale=<0 0 0>
   }
   n2 = "Transform (Modify) Bone - Bone: Barbell_04"!(ComponentPose=@n1, bAlphaBoolEnabled:bool=Get disk_3() -> (disk_3:bool), Translation="0,0,0", Rotation="0,0,0", Scale="0,0,0") -> (Pose) {
    alphaInputType=Bool
    scaleMode=BMM_Replace
    scaleSpace=BCS_BoneSpace
    translationMode=BMM_Additive
    translationSpace=BCS_BoneSpace
    boneToModify {boneName=Barbell_04}
    scale=<0 0 0>
   }
   n3 = "Transform (Modify) Bone - Bone: Barbell_02"!(ComponentPose=@n2, bAlphaBoolEnabled:bool=Get disk_2() -> (disk_2:bool), Translation="0,0,0", Rotation="0,0,0", Scale="0,0,0") -> (Pose) {
    alphaInputType=Bool
    scaleMode=BMM_Replace
    scaleSpace=BCS_BoneSpace
    translationMode=BMM_Additive
    translationSpace=BCS_BoneSpace
    boneToModify {boneName=Barbell_02}
    scale=<0 0 0>
   }
   n4 = Component To Local(ComponentPose=@n3)
   n5 = Use cached pose 'sim for Fe/game'()
   n6 = "Transform (Modify) Bone - Bone: Barbell_03"(
    ComponentPose="Transform (Modify) Bone - Bone: pack"(ComponentPose=Local To Component(LocalPose=@n5), Translation="0,3,0", Rotation="0,0,0") {translationMode=BMM_Additive, translationSpace=BCS_BoneSpace, boneToModify {boneName=pack}, translation {y=3}}
    bAlphaBoolEnabled=True
    Translation="0,0,0"
    Rotation="0,0,0"
    Scale="1,1,1"
   ) {
    alphaInputType=Bool
    scaleMode=BMM_Replace
    scaleSpace=BCS_BoneSpace
    translationMode=BMM_Additive
    translationSpace=BCS_BoneSpace
    boneToModify {boneName=Barbell_03}
   }
   n7 = "Transform (Modify) Bone - Bone: Barbell_04"(ComponentPose=@n6, bAlphaBoolEnabled=True, Translation="0,0,0", Rotation="0,0,0", Scale="1,1,1") {
    alphaInputType=Bool
    scaleMode=BMM_Replace
    scaleSpace=BCS_BoneSpace
    translationMode=BMM_Additive
    translationSpace=BCS_BoneSpace
    boneToModify {boneName=Barbell_04}
   }
   n8 = "Transform (Modify) Bone - Bone: Barbell_02"(ComponentPose=@n7, bAlphaBoolEnabled=True, Translation="0,0,0", Rotation="0,0,0", Scale="1,1,1") {
    alphaInputType=Bool
    scaleMode=BMM_Replace
    scaleSpace=BCS_BoneSpace
    translationMode=BMM_Additive
    translationSpace=BCS_BoneSpace
    boneToModify {boneName=Barbell_02}
   }
   n9 = Component To Local(ComponentPose=@n8)
   n10 = Get bShouldDisplayFrontendReactivity() -> (bShouldDisplayFrontendReactivity:bool)
   n11 = "Transform (Modify) Bone - Bone: pack"(
    ComponentPose=Local To Component(LocalPose=Use cached pose 'sim for Fe/game'())
    Translation=Get backpack_translation()
    Rotation=Get backpack_rotate()
   ) {rotationMode=BMM_Additive, translationMode=BMM_Additive, alphaBoolBlend {blendOption=Linear}, boneToModify {boneName=pack}}
   n12 = "Transform (Modify) Bone - Bone: Barbell_03"(ComponentPose=@n11, bAlphaBoolEnabled=Get disk_1(), Translation="0,0,0", Rotation="0,0,0", Scale="0,0,0") {
    alphaInputType=Bool
    scaleMode=BMM_Replace
    scaleSpace=BCS_BoneSpace
    translationMode=BMM_Additive
    translationSpace=BCS_BoneSpace
    boneToModify {boneName=Barbell_03}
    scale=<0 0 0>
   }
   n13 = "Transform (Modify) Bone - Bone: Barbell_04"(ComponentPose=@n12, bAlphaBoolEnabled=Get disk_3(), Translation="0,0,0", Rotation="0,0,0", Scale="0,0,0") {
    alphaInputType=Bool
    scaleMode=BMM_Replace
    scaleSpace=BCS_BoneSpace
    translationMode=BMM_Additive
    translationSpace=BCS_BoneSpace
    boneToModify {boneName=Barbell_04}
    scale=<0 0 0>
   }
   n14 = "Transform (Modify) Bone - Bone: Barbell_02"(ComponentPose=@n13, bAlphaBoolEnabled=Get disk_2(), Translation="0,0,0", Rotation="0,0,0", Scale="0,0,0") {
    alphaInputType=Bool
    scaleMode=BMM_Replace
    scaleSpace=BCS_BoneSpace
    translationMode=BMM_Additive
    translationSpace=BCS_BoneSpace
    boneToModify {boneName=Barbell_02}
    scale=<0 0 0>
   }
   Output Pose(Result=Blend Poses by bool(
     BlendPose_0=Blend Poses by bool(BlendPose_0=@n4, BlendPose_1=@n9, bActiveValue=@n10)
     BlendPose_1=Component To Local(ComponentPose=@n14)
     bActiveValue=Get bIsInFrontEnd()
    )) {layerGroup=None}
   Save cached pose 'sim for Fe/game'(Pose=Input Pose()) {cacheName=sim for Fe/game}
  ]}
 EventGraph {nodes [
   FE anim() {is=title K2Node_CustomEvent}
   > Branch(Condition="integer < integer"(A=Get killcount() -> (killcount:int=0), B=4))
   > n1 = Increment Int(Value=Get killcount() -> (killcount=0))
   Set Timer by Function Name!(
    execute=[@n7.then, @n1."  "]
    Object:Object Reference
    FunctionName:String=FE anim
    Time:float=Select!(Option 0:float=1, Option 1:float=8, Option 2:float=8, Index:int=Get killcount() -> (killcount=0)) -> (ReturnValue:float)
    bLooping:bool=false
    bMaxOncePerFrame:bool=false
    InitialStartDelay:float=0
    InitialStartDelayVariance:float=0
   ) -> (then, ReturnValue:Timer Handle Structure) {is=title K2Node_CallFunction}
   n2 = Event BlueprintUpdateAnimation()
   > Set Delta Time X(Delta Time X=@n2.DeltaTimeX)
   > Sequence()
   n3 = GetOwnerSkeletalMeshComponent()
   n4 = BreakRotator(InRot=BreakTransform(InTransform=GetDeltaTransformFromRefPose(self=@n3, BoneName=spine_05, BaseName=spine_01)).Rotation)
   n5 = MapRangeClamped(Value=@n4.Yaw, InRangeA=-10, InRangeB=-50, OutRangeB=52)
   > Set backpack_rotate(backpack_rotate=MakeRotator(Roll=@n5, Pitch=float / float(A=@n5, B=15), Yaw=-5))
   > Set backpack_translation(backpack_translation=MakeVector(Y=MapRangeClamped(Value=@n4.Yaw, InRangeA=-15, InRangeB=-50, OutRangeB=-11), Z=-1))
   n6 = Get killcount() -> (killcount=0)
   > Set disk_1(disk_1:bool=NOT Boolean(A="integer >= integer"(A=@n6, B=1))) -> (Output_Get:bool)
   > Set disk_2(disk_2:bool=NOT Boolean(A="integer >= integer"(A=@n6, B=2))) -> (Output_Get:bool)
   > Set disk_3(disk_3:bool=NOT Boolean(A="integer >= integer"(A=@n6, B=3))) -> (Output_Get:bool)
   Event BlueprintBeginPlay() {is=title K2Node_Event}
   > n7 = Branch(Condition=Get bIsInFrontEnd())
   n8 = Event BlueprintProcessAnimInputEvent()
   > Delay(Duration=0.05)
   n9 = Get killcount() -> (killcount=0)
   > Branch(Condition=AND Boolean(A=IsNamed(self=@n8.AnimInputEvent, InEventName=aie.EnemyPlayerEliminated), B="integer < integer"(A=@n9, B=4)))
   > Increment Int(Value=Get killcount() -> (killcount=0))
   n10 = Get killcount() -> (killcount=0)
   Branch(Condition=AND Boolean(A=IsNamed(InEventName=aie.EnemyPlayerEliminated), B="integer < integer"(A=@n10, B=4)))
   > Increment Int(Value=Get killcount() -> (killcount=0))
   n11 = Get killcount() -> (killcount=0)
   Set disk_1(disk_1=NOT Boolean(A="integer >= integer"(A=@n11, B=1)))
   > Set disk_2(disk_2=NOT Boolean(A="integer >= integer"(A=@n11, B=2)))
   > Set disk_3(disk_3=NOT Boolean(A="integer >= integer"(A=@n11, B=3)))
   TryGetPawnOwner()
  ]}
]
variables [
 backpack_rotate {type=Rotator}
 backpack_translation {type=Vector}
 Delta Time X {type=number}
 disk_1 {type=boolean}
 disk_2 {type=boolean}
 disk_3 {type=boolean}
 FE_withpawn {type=boolean}
 killcount {type=integer}
 killcount_1 {type=integer}
 killcount_2 {type=integer}
 killcount_3 {type=integer}
 run anim {type=boolean}
]
