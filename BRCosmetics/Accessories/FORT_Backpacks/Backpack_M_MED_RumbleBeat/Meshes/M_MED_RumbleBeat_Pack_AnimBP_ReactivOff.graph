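# Anim Blueprint for the RumbleBeat backpack, "ReactivOff" variant, that offsets the pack to follow the wearer's spine.
# Each update maps spine_05-versus-spine_01 yaw, relative to ref pose, through clamped ranges into backpack rotation and translation.
# The AnimGraph adds these and a fixed bone-space nudge to the pack bone over the input pose, then holds three Barbell bones at unit scale.
# Kill-count, disk and reactivity variables are declared but unused.
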AnimBlueprint /BRCosmetics/Accessories/FORT_Backpacks/Backpack_M_MED_RumbleBeat/Meshes/M_MED_RumbleBeat_Pack_AnimBP_ReactivOff
target_skeleton=&/M_MED_RumbleBeat_Pack_Skeleton
graphs [
 AnimGraph {nodes [
   n1 = "Transform (Modify) Bone - Bone: pack"(ComponentPose=Local To Component(LocalPose=Input Pose()), Translation=Get backpack_translation(), Rotation=Get backpack_rotate()) {rotationMode=BMM_Additive, translationMode=BMM_Additive, alphaBoolBlend {blendOption=Linear}, boneToModify {boneName=pack}}
   n2 = "Transform (Modify) Bone - Bone: Barbell_03"!(
    ComponentPose="Transform (Modify) Bone - Bone: pack"(ComponentPose=@n1, Translation="0,3,0", Rotation="0,0,0") {translationMode=BMM_Additive, translationSpace=BCS_BoneSpace, boneToModify {boneName=pack}, translation {y=3}}
    bAlphaBoolEnabled:bool=True
    Translation="0,0,0"
    Rotation="0,0,0"
    Scale="1,1,1"
   ) -> (Pose) {
    alphaInputType=Bool
    scaleMode=BMM_Replace
    scaleSpace=BCS_BoneSpace
    translationMode=BMM_Additive
    translationSpace=BCS_BoneSpace
    boneToModify {boneName=Barbell_03}
   }
   n3 = "Transform (Modify) Bone - Bone: Barbell_04"!(ComponentPose=@n2, bAlphaBoolEnabled:bool=True, Translation="0,0,0", Rotation="0,0,0", Scale="1,1,1") -> (Pose) {
    alphaInputType=Bool
    scaleMode=BMM_Replace
    scaleSpace=BCS_BoneSpace
    translationMode=BMM_Additive
    translationSpace=BCS_BoneSpace
    boneToModify {boneName=Barbell_04}
   }
   n4 = "Transform (Modify) Bone - Bone: Barbell_02"!(ComponentPose=@n3, bAlphaBoolEnabled:bool=True, Translation="0,0,0", Rotation="0,0,0", Scale="1,1,1") -> (Pose) {
    alphaInputType=Bool
    scaleMode=BMM_Replace
    scaleSpace=BCS_BoneSpace
    translationMode=BMM_Additive
    translationSpace=BCS_BoneSpace
    boneToModify {boneName=Barbell_02}
   }
   Output Pose(Result=Component To Local(ComponentPose=@n4)) {layerGroup=None}
  ]}
 EventGraph {nodes [
   n1 = Event BlueprintUpdateAnimation()
   > Set Delta Time X(Delta Time X=@n1.DeltaTimeX)
   > Sequence()
   n2 = GetOwnerSkeletalMeshComponent()
   n3 = BreakRotator(InRot=BreakTransform(InTransform=GetDeltaTransformFromRefPose(self=@n2, BoneName=spine_05, BaseName=spine_01)).Rotation)
   n4 = MapRangeClamped(Value=@n3.Yaw, InRangeA=-10, InRangeB=-50, OutRangeB=52)
   > Set backpack_rotate(backpack_rotate=MakeRotator(Roll=@n4, Pitch=float / float(A=@n4, B=15), Yaw=-5))
   > Set backpack_translation(backpack_translation=MakeVector(Y=MapRangeClamped(Value=@n3.Yaw, InRangeA=-15, InRangeB=-50, OutRangeB=-11), Z=-1))
   TryGetPawnOwner()
  ]}
]
variables [
 As B Backpack Last Voice Dive {type=B_Backpack_LastVoiceDive_C, default=null}
 backpack_rotate {type=Rotator}
 backpack_translation {type=Vector}
 CosmeticVariantReactivity {type=boolean}
 Delta Time X {type=number}
 disk_1 {type=boolean}
 disk_2 {type=boolean}
 disk_3 {type=boolean}
 FE_withpawn {type=boolean}
 killcount {type=integer}
 killcount_1 {type=integer}
 killcount_2 {type=integer}
 killcount_3 {type=integer}
 run anim {type=boolean}
]
